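; Passive quest ability that credits the player for affecting three unique trees (organic props) within one swinging movement, going by its name and tags.
; On activation it listens for a configured gameplay event on the owning pawn and collects targets tagged as organic props in a unique array.
; A count of at least three sends a complex custom stat event with additional source tags to the quest manager.
; Removal of the Swinging movement tag clears the array.
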(event Ability|EventOnEndAbility
  (params (bWasCancelled :type Boolean))
  (Ability|EndAbility))

(event Ability|EventFailedToActivatePassiveAbility
  (params (ActorInfo :type Gameplay%20Ability%20Actor%20Info%20Structure))
  (Ability|EndAbility))

(event Ability|EventActivateAbility
  (bind _unique_trees (Variables|Default|GetUnique%20Trees))
  (|CallParentFunction)
  (bind (_asynctaskproxy _payload) (Ability|Async|WaitGameplayEventToActor (Variables|GAVars|Data|GetOwningPlayerPawn) (Variables|Default|GetEvent%20Tag))
    (:then
      (Ability|Tasks|WaitGameplayTagRemove "(TagName=\"MovementMode.Swinging\")" (Variables|GAVars|Data|GetOwningPlayerPawn)
        (:Removed
          (Utilities|Array|Clear (Variables|Default|GetUnique%20Trees)))))
    (:EventReceived
      (bind _returnvalue (GameplayTags|HasMatchingGameplayTag(Message) (.Target (Utilities|Struct|BreakGameplayEventData _payload)) "(TagName=\"Building.Type.Prop.Organic\")"))
      (if _returnvalue
        (Utilities|Array|AddUnique _unique_trees (.Target (Utilities|Struct|BreakGameplayEventData _payload)))
        (if (>= (Utilities|Array|Length _unique_trees) 3)
          (Quest|SendComplexCustomStatEvent (Variables|GAVars|Data|GetOwningPlayerQuestManager) 0 (Variables|Default|GetAdditional%20Source%20Tags) "(GameplayTags=)" 1))))))
